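# Sound cue for the first-person (1P) Juno version of the SpaceChimp emote music.
# A looping wave player plays the Juno loop through an Enveloper that applies a short volume fade-in at the start.
# It uses the 1P emote music sound class and emote music concurrency rules, and sends to a music audio-analysis submix.
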SoundCue /BRCosmetics/Sounds/Emotes/SpaceChimp/Emote_SpaceChimp_1P_Juno_Cue
root=Enveloper_0
nodes [
 {
  id=Enveloper_0
  children [WavePlayer_0]
  properties {pitchCurve {editorCurveData {keys [{time=0}]}}, volumeCurve {editorCurveData {keys [<at 0 0>,<at 0.1 1>]}}}
 }
 {id=WavePlayer_0, properties {bLooping=true, soundWaveAssetPtr=&/Emote_SpaceChimp_Loop_JunoVersion}}
]
properties {
 maxDistance=2097152
 soundClassObject=/Game/Sounds/FortSoundClassesAndModes/Fort_Music_Emote_1P_PSM
 volumeMultiplier=1.2
 concurrencySet [/Game/Sounds/Concurrency_Rules/Athena_EmoteMusic_Conc]
 soundSubmixSends [{minSendLevel=0, soundSubmix=&Fortnite_AudioAnalysis_Music}]
}
