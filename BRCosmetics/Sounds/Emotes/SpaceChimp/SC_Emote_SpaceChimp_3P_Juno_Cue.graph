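# Sound cue for the looping third-person music of the Space Chimp emote, Juno version.
# A looping wave player feeds an enveloper that ramps volume up briefly at the start.
# It uses Athena emote attenuation and concurrency with the third-person emote music sound class.
# It keeps playing when silent and sends to a music audio-analysis submix.
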SoundCue /BRCosmetics/Sounds/Emotes/SpaceChimp/SC_Emote_SpaceChimp_3P_Juno_Cue
root=Enveloper_1
nodes [
 {
  id=Enveloper_1
  children [WavePlayer_0]
  properties {pitchCurve {editorCurveData {keys [{time=0}]}}, volumeCurve {editorCurveData {keys [<at 0 0>,<at 0.1 1>]}}}
 }
 {id=WavePlayer_0, properties {bLooping=true, soundWaveAssetPtr=&/Emote_SpaceChimp_Loop_JunoVersion}}
]
properties {
 attenuationSettings=&Fort_Emotes_Athena_Att
 maxDistance=1600
 soundClassObject=/Game/Sounds/FortSoundClassesAndModes/Fort_Music_Emote_3P
 virtualizationMode=PlayWhenSilent
 volumeMultiplier=1.2
 concurrencySet [/Game/Sounds/Concurrency_Rules/Athena_EmoteMusic_Conc]
 soundSubmixSends [{minSendLevel=0, soundSubmix=&Fortnite_AudioAnalysis_Music}]
}
